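# Anim Blueprint for the ClimbSpill cosmetic backpack, adding secondary motion to its head, elbow, foot, flip-flop and bow bones with chained AnimDynamics.
# The cached input pose feeds head and elbow chains: one toggles head limits on bIsMoving2D, another blends in while skydiving.
# That result continues through feet, flip-flops and bow, then blends against a separately tuned full chain for the front end.
# The event graph holds only unconnected stub nodes.
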AnimBlueprint /BRCosmetics/Accessories/FORT_Backpacks/Backpack_M_MED_ClimbSpill/Meshes/M_MED_ClimbSpill_Pack_AnimBP
target_skeleton=&/M_MED_ClimbSpill_Pack_Skeleton
graphs [
 AnimGraph {nodes [
   n1 = Get bIsMoving2D()
   n2 = Anim Dynamics - Bone: dyn_head_jnt!(
    ComponentPose=Local To Component(LocalPose=Use cached pose 'SavedInputPose'())
    bAlphaBoolEnabled:bool=NOT Boolean(A=@n1)
    ExternalForce="0,0,0"
   ) -> (Pose) {
    alphaInputType=Bool
    angularDampingOverride=1
    angularSpringConstant=2500
    bOverrideAngularDamping=true
    bUseGravityOverride=true
    alphaBoolBlend {blendInTime=0.2, blendOption=Linear, blendOutTime=0.2}
    boundBone {boneName=dyn_head_jnt}
    componentAppliedLinearAccClamp=<1000 1000 1000>
    componentLinearAccScale=<0.3 0.3 0.3>
    componentLinearVelScale=<0.4 0.4 0.4>
    gravityOverride {z=-600}
    physicsBodyDefinitions [{
      boundBone {boneName=dyn_head_jnt}
      boxExtents=<6 10 10>
      constraintSetup {angularLimitsMax=<7 10 10>, angularLimitsMin=<-7 -10 0>, angularTarget=<0 0 0>}
      localJointOffset=<-1 0 0>
     }]
   }
   n3 = Anim Dynamics - Bone: dyn_head_jnt(ComponentPose=@n2, bAlphaBoolEnabled=@n1, ExternalForce="0,0,0") {
    alphaInputType=Bool
    angularDampingOverride=1
    angularSpringConstant=2500
    bOverrideAngularDamping=true
    bUseGravityOverride=true
    alphaBoolBlend {blendInTime=0.3, blendOption=Linear, blendOutTime=0.3}
    boundBone {boneName=dyn_head_jnt}
    componentAppliedLinearAccClamp=<1000 1000 1000>
    componentLinearAccScale=<0.3 0.3 0.3>
    componentLinearVelScale=<0.4 0.4 0.4>
    gravityOverride {z=-600}
    physicsBodyDefinitions [{
      boundBone {boneName=dyn_head_jnt}
      boxExtents=<6 10 10>
      constraintSetup {angularLimitsMax=<7 10 10>, angularLimitsMin=<-7 -10 -10>, angularTarget=<0 0 0>}
      localJointOffset=<-1 0 0>
     }]
   }
   n4 = Anim Dynamics - Bone: dyn_elbow_l_jnt(ComponentPose=@n3, Alpha=1, ExternalForce="0,0,0") {
    angularDampingOverride=1
    angularSpringConstant=800
    bAngularSpring=true
    bOverrideAngularDamping=true
    bUseGravityOverride=true
    gravityScale=0
    alphaBoolBlend {blendOption=Linear}
    boundBone {boneName=dyn_elbow_l_jnt}
    componentAppliedLinearAccClamp=<900 900 900>
    componentLinearAccScale=<0.3 0.4 0.2>
    componentLinearVelScale=<0.4 0.4 0.3>
    gravityOverride {z=-50}
    physicsBodyDefinitions [{
      boundBone {boneName=dyn_elbow_l_jnt}
      boxExtents=<5 5 5>
      constraintSetup {angularLimitsMax=<5 3 10>, angularLimitsMin=<-5 -3 -5>, angularTarget=<1 0 0>}
      localJointOffset=<1 0 0>
     }]
   }
   n5 = Anim Dynamics - Bone: dyn_elbow_r_jnt(ComponentPose=@n4, Alpha=1, ExternalForce="0,0,0") {
    angularDampingOverride=1
    angularSpringConstant=800
    bAngularSpring=true
    bOverrideAngularDamping=true
    bUseGravityOverride=true
    gravityScale=0
    alphaBoolBlend {blendOption=Linear}
    boundBone {boneName=dyn_elbow_r_jnt}
    componentAppliedLinearAccClamp=<900 900 900>
    componentLinearAccScale=<0.3 0.4 0.2>
    componentLinearVelScale=<0.4 0.4 0.3>
    gravityOverride {z=-50}
    physicsBodyDefinitions [{
      boundBone {boneName=dyn_elbow_r_jnt}
      boxExtents=<7 7 7>
      constraintSetup {angularLimitsMax=<5 3 10>, angularLimitsMin=<-5 -3 -10>, angularTarget=<1 0 0>}
      localJointOffset=<-1 0 0>
     }]
   }
   n6 = Component To Local(ComponentPose=@n5)
   n7 = Anim Dynamics - Bone: dyn_head_jnt(ComponentPose=Local To Component(LocalPose=Use cached pose 'SavedInputPose'()), bAlphaBoolEnabled=True, ExternalForce="0,0,0") {
    alphaInputType=Bool
    angularDampingOverride=1
    angularSpringConstant=2500
    bOverrideAngularDamping=true
    bUseGravityOverride=true
    alphaBoolBlend {blendInTime=0.3, blendOption=Linear, blendOutTime=0.3}
    boundBone {boneName=dyn_head_jnt}
    componentAppliedLinearAccClamp=<10000 10000 10000>
    componentLinearAccScale=<0.1 0.1 0.1>
    componentLinearVelScale=<0.1 0.1 0.1>
    gravityOverride {z=-500}
    physicsBodyDefinitions [{
      boundBone {boneName=dyn_head_jnt}
      boxExtents=<6 10 10>
      constraintSetup {angularLimitsMax=<5 5 5>, angularLimitsMin=<-5 -5 -5>, angularTarget=<0 0 0>}
      localJointOffset=<-1 0 0>
     }]
   }
   n8 = Anim Dynamics - Bone: dyn_elbow_l_jnt(ComponentPose=@n7, Alpha=1, ExternalForce="0,0,0") {
    angularDampingOverride=1
    angularSpringConstant=800
    bAngularSpring=true
    bGravityOverrideInSimSpace=true
    bOverrideAngularDamping=true
    bUseGravityOverride=true
    gravityScale=0
    alphaBoolBlend {blendOption=Linear}
    boundBone {boneName=dyn_elbow_l_jnt}
    componentAppliedLinearAccClamp=<900 900 900>
    componentLinearAccScale=<0.3 0.4 0.2>
    componentLinearVelScale=<0.4 0.4 0.3>
    gravityOverride {z=-50}
    physicsBodyDefinitions [{
      boundBone {boneName=dyn_elbow_l_jnt}
      boxExtents=<5 5 5>
      constraintSetup {angularLimitsMax=<5 3 0>, angularLimitsMin=<-5 -3 -5>, angularTarget=<1 0 0>}
      localJointOffset=<1 0 0>
     }]
   }
   n9 = Anim Dynamics - Bone: dyn_elbow_r_jnt(ComponentPose=@n8, Alpha=1, ExternalForce="0,0,0") {
    angularDampingOverride=1
    angularSpringConstant=800
    bAngularSpring=true
    bGravityOverrideInSimSpace=true
    bOverrideAngularDamping=true
    bUseGravityOverride=true
    gravityScale=0
    alphaBoolBlend {blendOption=Linear}
    boundBone {boneName=dyn_elbow_r_jnt}
    componentAppliedLinearAccClamp=<900 900 900>
    componentLinearAccScale=<0.3 0.4 0.2>
    componentLinearVelScale=<0.4 0.4 0.3>
    gravityOverride {z=-50}
    physicsBodyDefinitions [{
      boundBone {boneName=dyn_elbow_r_jnt}
      boxExtents=<5 5 5>
      constraintSetup {angularLimitsMax=<5 3 0>, angularLimitsMin=<-5 -3 -5>, angularTarget=<1 0 0>}
      localJointOffset=<-1 0 0>
     }]
   }
   n10 = Component To Local(ComponentPose=@n9)
   n11 = Get bIsSkyDiving()
   n12 = Anim Dynamics - Bone: dyn_foot_l_jnt(
    ComponentPose=Local To Component(LocalPose=Two Way Blend!(A=@n6, B=@n10, bAlphaBoolEnabled:bool=@n11) -> (Pose) {blendNode {alphaInputType=Bool}})
    Alpha=1
    ExternalForce="0,0,0"
   ) {
    angularDampingOverride=1
    angularSpringConstant=800
    bAngularSpring=true
    bGravityOverrideInSimSpace=true
    bOverrideAngularDamping=true
    bUseGravityOverride=true
    gravityScale=0
    alphaBoolBlend {blendOption=Linear}
    boundBone {boneName=dyn_foot_l_jnt}
    componentAppliedLinearAccClamp=<900 900 900>
    componentLinearAccScale=<0.3 0.4 0.2>
    componentLinearVelScale=<0.4 0.4 0.3>
    gravityOverride {z=-50}
    physicsBodyDefinitions [{
      boundBone {boneName=dyn_foot_l_jnt}
      boxExtents=<5 5 5>
      constraintSetup {angularLimitsMax=<0 1 1>, angularLimitsMin=<0 -1 -4>, angularTarget=<1 0 0>}
      localJointOffset=<1 0 0>
     }]
   }
   n13 = Anim Dynamics - Bone: dyn_foot_r_jnt(ComponentPose=@n12, Alpha=1, ExternalForce="0,0,0") {
    angularDampingOverride=1
    angularSpringConstant=800
    bAngularSpring=true
    bOverrideAngularDamping=true
    bUseGravityOverride=true
    gravityScale=0
    alphaBoolBlend {blendOption=Linear}
    boundBone {boneName=dyn_foot_r_jnt}
    componentAppliedLinearAccClamp=<900 900 900>
    componentLinearAccScale=<0.3 0.4 0.2>
    componentLinearVelScale=<0.4 0.4 0.3>
    gravityOverride {z=-50}
    physicsBodyDefinitions [{
      boundBone {boneName=dyn_foot_r_jnt}
      boxExtents=<5 5 5>
      constraintSetup {angularLimitsMax=<0 1 1>, angularLimitsMin=<0 -1 -4>, angularTarget=<1 0 0>}
      localJointOffset=<-1 0 0>
     }]
   }
   n14 = Anim Dynamics - Bone: dyn_flip_flops_l_jnt(ComponentPose=@n13, Alpha=1, ExternalForce="0,0,0") {
    angularDampingOverride=1
    angularSpringConstant=800
    bAngularSpring=true
    bOverrideAngularDamping=true
    bUseGravityOverride=true
    gravityScale=0
    alphaBoolBlend {blendOption=Linear}
    boundBone {boneName=dyn_flip_flops_l_jnt}
    componentAppliedLinearAccClamp=<900 900 900>
    componentLinearAccScale=<0.3 0.4 0.2>
    componentLinearVelScale=<0.4 0.4 0.3>
    gravityOverride {z=-50}
    physicsBodyDefinitions [{
      boundBone {boneName=dyn_flip_flops_l_jnt}
      boxExtents=<5 5 5>
      constraintSetup {angularLimitsMax=<0 1 10>, angularLimitsMin=<0 -1 -12>, angularTarget=<1 0 0>}
      localJointOffset=<1 0 0>
     }]
   }
   n15 = Anim Dynamics - Bone: dyn_flip_flops_r_jnt(ComponentPose=@n14, Alpha=1, ExternalForce="0,0,0") {
    angularDampingOverride=1
    angularSpringConstant=800
    bAngularSpring=true
    bGravityOverrideInSimSpace=true
    bOverrideAngularDamping=true
    bUseGravityOverride=true
    gravityScale=0
    alphaBoolBlend {blendOption=Linear}
    boundBone {boneName=dyn_flip_flops_r_jnt}
    componentAppliedLinearAccClamp=<800 800 800>
    componentLinearAccScale=<0.3 0.4 0.2>
    componentLinearVelScale=<0.4 0.5 0.3>
    gravityOverride {z=-50}
    physicsBodyDefinitions [{
      boundBone {boneName=dyn_flip_flops_r_jnt}
      boxExtents=<5 5 5>
      constraintSetup {angularLimitsMax=<0 2 12>, angularLimitsMin=<0 -2 -10>, angularTarget=<1 0 0>}
      localJointOffset=<-1 0 0>
     }]
   }
   n16 = Anim Dynamics - Bone: dyn_bow_jnt(ComponentPose=@n15, Alpha=1, ExternalForce="0,0,0") {
    angularDampingOverride=0.98
    angularSpringConstant=600
    bAngularSpring=true
    bOverrideAngularDamping=true
    bUseGravityOverride=true
    gravityScale=0
    alphaBoolBlend {blendOption=Linear}
    boundBone {boneName=dyn_bow_jnt}
    componentAppliedLinearAccClamp=<5000 5000 5000>
    componentLinearAccScale=<0.1 0.1 0.1>
    componentLinearVelScale=<0.1 0.1 0.1>
    gravityOverride {z=-600}
    physicsBodyDefinitions [{
      boundBone {boneName=dyn_bow_jnt}
      boxExtents=<3 5 5>
      constraintSetup {angularLimitsMax=<5 5 5>, angularLimitsMin=<-5 -5 -5>, angularTarget=<1 0 0>}
      localJointOffset=<1 0 0>
     }]
   }
   n17 = Anim Dynamics - Bone: dyn_head_jnt(ComponentPose=Local To Component(LocalPose=Use cached pose 'SavedInputPose'()), bAlphaBoolEnabled=True, ExternalForce="0,0,0") {
    alphaInputType=Bool
    angularDampingOverride=1
    angularSpringConstant=2500
    bGravityOverrideInSimSpace=true
    bOverrideAngularDamping=true
    bUseGravityOverride=true
    alphaBoolBlend {blendInTime=0.2, blendOption=Linear, blendOutTime=0.2}
    boundBone {boneName=dyn_head_jnt}
    componentAppliedLinearAccClamp=<10000 10000 10000>
    componentLinearAccScale=<0.3 0.3 0.3>
    componentLinearVelScale=<0.3 0.3 0.3>
    gravityOverride {z=-600}
    physicsBodyDefinitions [{
      boundBone {boneName=dyn_head_jnt}
      boxExtents=<9 9 9>
      constraintSetup {angularLimitsMax=<7 10 10>, angularLimitsMin=<-7 -10 0>, angularTarget=<0 0 0>}
      localJointOffset=<-1 0 0>
     }]
   }
   n18 = Anim Dynamics - Bone: dyn_elbow_l_jnt(ComponentPose=@n17, Alpha=1, ExternalForce="0,0,0") {
    angularDampingOverride=1
    angularSpringConstant=800
    bAngularSpring=true
    bGravityOverrideInSimSpace=true
    bOverrideAngularDamping=true
    bUseGravityOverride=true
    gravityScale=0
    alphaBoolBlend {blendOption=Linear}
    boundBone {boneName=dyn_elbow_l_jnt}
    componentAppliedLinearAccClamp=<1000 1000 1000>
    componentLinearAccScale=<0.1 0.1 0.1>
    componentLinearVelScale=<0.1 0.1 0.1>
    gravityOverride {z=-50}
    physicsBodyDefinitions [{
      boundBone {boneName=dyn_elbow_l_jnt}
      boxExtents=<4 4 4>
      constraintSetup {angularLimitsMax=<5 3 0>, angularLimitsMin=<-5 -3 -5>, angularTarget=<1 0 0>}
      localJointOffset=<1 0 0>
     }]
   }
   n19 = Anim Dynamics - Bone: dyn_elbow_r_jnt(ComponentPose=@n18, Alpha=1, ExternalForce="0,0,0") {
    angularDampingOverride=1
    angularSpringConstant=800
    bAngularSpring=true
    bGravityOverrideInSimSpace=true
    bOverrideAngularDamping=true
    bUseGravityOverride=true
    gravityScale=0
    alphaBoolBlend {blendOption=Linear}
    boundBone {boneName=dyn_elbow_r_jnt}
    componentAppliedLinearAccClamp=<1000 1000 1000>
    componentLinearAccScale=<0.2 0.2 0.2>
    componentLinearVelScale=<0.2 0.2 0.2>
    gravityOverride {z=-50}
    physicsBodyDefinitions [{
      boundBone {boneName=dyn_elbow_r_jnt}
      boxExtents=<5 5 5>
      constraintSetup {angularLimitsMax=<5 3 0>, angularLimitsMin=<-5 -3 -5>, angularTarget=<1 0 0>}
      localJointOffset=<-1 0 0>
     }]
   }
   n20 = Anim Dynamics - Bone: dyn_foot_l_jnt(ComponentPose=@n19, Alpha=1, ExternalForce="0,0,0") {
    angularDampingOverride=1
    angularSpringConstant=800
    bAngularSpring=true
    bGravityOverrideInSimSpace=true
    bOverrideAngularDamping=true
    bUseGravityOverride=true
    gravityScale=0
    alphaBoolBlend {blendOption=Linear}
    boundBone {boneName=dyn_foot_l_jnt}
    componentAppliedLinearAccClamp=<900 900 900>
    componentLinearAccScale=<0.3 0.4 0.2>
    componentLinearVelScale=<0.4 0.4 0.3>
    gravityOverride {z=-50}
    physicsBodyDefinitions [{
      boundBone {boneName=dyn_foot_l_jnt}
      boxExtents=<5 5 5>
      constraintSetup {angularLimitsMax=<0 1 1>, angularLimitsMin=<0 -1 -4>, angularTarget=<1 0 0>}
      localJointOffset=<1 0 0>
     }]
   }
   n21 = Anim Dynamics - Bone: dyn_foot_r_jnt(ComponentPose=@n20, Alpha=1, ExternalForce="0,0,0") {
    angularDampingOverride=1
    angularSpringConstant=800
    bAngularSpring=true
    bOverrideAngularDamping=true
    bUseGravityOverride=true
    gravityScale=0
    alphaBoolBlend {blendOption=Linear}
    boundBone {boneName=dyn_foot_r_jnt}
    componentAppliedLinearAccClamp=<900 900 900>
    componentLinearAccScale=<0.3 0.4 0.2>
    componentLinearVelScale=<0.4 0.4 0.3>
    gravityOverride {z=-50}
    physicsBodyDefinitions [{
      boundBone {boneName=dyn_foot_r_jnt}
      boxExtents=<5 5 5>
      constraintSetup {angularLimitsMax=<0 1 1>, angularLimitsMin=<0 -1 -4>, angularTarget=<1 0 0>}
      localJointOffset=<-1 0 0>
     }]
   }
   n22 = Anim Dynamics - Bone: dyn_flip_flops_l_jnt(ComponentPose=@n21, Alpha=1, ExternalForce="0,0,0") {
    angularDampingOverride=1
    angularSpringConstant=800
    bAngularSpring=true
    bOverrideAngularDamping=true
    bUseGravityOverride=true
    gravityScale=0
    alphaBoolBlend {blendOption=Linear}
    boundBone {boneName=dyn_flip_flops_l_jnt}
    componentAppliedLinearAccClamp=<900 900 900>
    componentLinearAccScale=<0.1 0.1 0.1>
    componentLinearVelScale=<0.1 0.1 0.1>
    gravityOverride {z=-50}
    physicsBodyDefinitions [{
      boundBone {boneName=dyn_flip_flops_l_jnt}
      boxExtents=<4 4 4>
      constraintSetup {angularLimitsMax=<0 1 10>, angularLimitsMin=<0 -1 -12>, angularTarget=<1 0 0>}
      localJointOffset=<1 0 0>
     }]
   }
   n23 = Anim Dynamics - Bone: dyn_flip_flops_r_jnt(ComponentPose=@n22, Alpha=1, ExternalForce="0,0,0") {
    angularDampingOverride=1
    angularSpringConstant=800
    bAngularSpring=true
    bGravityOverrideInSimSpace=true
    bOverrideAngularDamping=true
    bUseGravityOverride=true
    gravityScale=0
    alphaBoolBlend {blendOption=Linear}
    boundBone {boneName=dyn_flip_flops_r_jnt}
    componentAppliedLinearAccClamp=<1000 1000 1000>
    componentLinearAccScale=<0.1 0.1 0.1>
    componentLinearVelScale=<0.1 0.1 0.1>
    gravityOverride {z=-50}
    physicsBodyDefinitions [{
      boundBone {boneName=dyn_flip_flops_r_jnt}
      boxExtents=<5 5 5>
      constraintSetup {angularLimitsMax=<0 2 12>, angularLimitsMin=<0 -2 -10>, angularTarget=<1 0 0>}
      localJointOffset=<-1 0 0>
     }]
   }
   n24 = Anim Dynamics - Bone: dyn_bow_jnt(ComponentPose=@n23, Alpha=1, ExternalForce="0,0,0") {
    angularDampingOverride=1
    angularSpringConstant=600
    bAngularSpring=true
    bOverrideAngularDamping=true
    bUseGravityOverride=true
    gravityScale=0
    alphaBoolBlend {blendOption=Linear}
    boundBone {boneName=dyn_bow_jnt}
    componentAppliedLinearAccClamp=<1000 1000 1001>
    componentLinearAccScale=<0.1 0.1 0.1>
    componentLinearVelScale=<0.1 0.1 0.1>
    gravityOverride {z=-600}
    physicsBodyDefinitions [{
      boundBone {boneName=dyn_bow_jnt}
      boxExtents=<3 5 5>
      constraintSetup {angularLimitsMax=<5 5 5>, angularLimitsMin=<-5 -5 -5>, angularTarget=<1 0 0>}
      localJointOffset=<1 0 0>
     }]
   }
   Output Pose(Result=Two Way Blend(A=Component To Local(ComponentPose=@n16), B=Component To Local(ComponentPose=@n24), bAlphaBoolEnabled=Get bIsInFrontEnd()) {blendNode {alphaInputType=Bool}}) {layerGroup=None}
   Save cached pose 'SavedInputPose'(Pose=Input Pose()) {cacheName=SavedInputPose}
  ]}
 EventGraph {nodes [
   TryGetPawnOwner()
   Event BlueprintUpdateAnimation()
  ]}
]
variables []
